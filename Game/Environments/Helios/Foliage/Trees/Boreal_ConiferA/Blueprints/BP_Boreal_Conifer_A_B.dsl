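; BP_Boreal_Conifer_A_B is a Blueprint for a boreal conifer tree variant in the Helios environment foliage.
; Its custom Resize event randomizes actor scale independently per axis within a reduced range, so proportions can be non-uniform.
; Nothing in this file calls Resize.
; BeginPlay, overlap and Tick events are empty, and the construction script only calls the parent's.
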(event Custom|Resize
  (Transformation|SetActorScale3D self (Math|Vector|MakeVector (Math|Random|RandomFloatinRange 0.6 1.0) (Math|Random|RandomFloatinRange 0.6 1.0) (Math|Random|RandomFloatinRange 0.6 1.0))))

(event EventBeginPlay)

(event Collision|EventActorBeginOverlap
  (params (OtherActor :type Actor%20Object%20Reference)))

(event EventTick
  (params (DeltaSeconds :type Float%20%28single-precision%29)))

(fn ConstructionScript
  (|CallParentFunction))
